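# Anim Blueprint for the F_MED_FermiumStraw_Dust face accessory, animating the character's hair on a pose copied from the parent mesh.
# The ponytail uses either RigidBody nodes with BrightIon's physics asset or an AnimDynamics chain, chosen by bRigidBodyEnabled.
# AnimDynamics then drives bangs and side strands, with additive bone offsets posing the tail for movement states and back bling clearance.
# The event graph derives those states, head-twist alpha and zipline status.
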AnimBlueprint /BRCosmetics/Characters/Player/Female/Medium/Bodies/F_MED_FermiumStraw_Dust/Meshes/Parts/F_MED_FermiumStraw_Dust_FaceAcc_AnimBP
target_skeleton=&/F_MED_FermiumStraw_Dust_FaceAcc_Skeleton
graphs [
 AnimGraph {nodes [
   n1 = Get bIsSkyDiving()
   n2 = Get jump_falling() -> (jump_falling:bool)
   n3 = Get Crouch_moving() -> (Crouch_moving:bool)
   n4 = "Transform (Modify) Bone - Bone: dyn_hair_tail_01"!(
    ComponentPose=Local To Component(LocalPose=Use cached pose 'Aftersim'())
    bAlphaBoolEnabled:bool=@n3
    Translation="0,0,0"
    Rotation="0,0,35"
    Scale="1,1,1"
   ) -> (Pose) {
    alphaInputType=Bool
    rotationMode=BMM_Additive
    scaleMode=BMM_Additive
    translationMode=BMM_Additive
    alphaBoolBlend {blendInTime=0.1, blendOption=Linear, blendOutTime=0.2}
    boneToModify {boneName=dyn_hair_tail_01}
    rotation {roll=35}
   }
   n5 = "Transform (Modify) Bone - Bone: dyn_hair_tail_02"!(ComponentPose=@n4, bAlphaBoolEnabled:bool=@n3, Translation="0,0,0", Rotation="0,0,-15", Scale="1,1,1") -> (Pose) {
    alphaInputType=Bool
    rotationMode=BMM_Additive
    scaleMode=BMM_Additive
    translationMode=BMM_Additive
    alphaBoolBlend {blendInTime=0.1, blendOption=Linear, blendOutTime=0.2}
    boneToModify {boneName=dyn_hair_tail_02}
    rotation {roll=-15}
   }
   n6 = "Transform (Modify) Bone - Bone: dyn_hair_tail_03"!(ComponentPose=@n5, bAlphaBoolEnabled:bool=@n3, Translation="0,0,0", Rotation="0,0,25", Scale="1,1,1") -> (Pose) {
    alphaInputType=Bool
    rotationMode=BMM_Additive
    scaleMode=BMM_Additive
    translationMode=BMM_Additive
    alphaBoolBlend {blendInTime=0.1, blendOption=Linear, blendOutTime=0.2}
    boneToModify {boneName=dyn_hair_tail_03}
    rotation {roll=25}
   }
   n7 = "Transform (Modify) Bone - Bone: dyn_hair_tail_01"(ComponentPose=@n6, bAlphaBoolEnabled=Get bIsSkydiveDiveMode(), Translation="0,0,0", Rotation="0,0,50", Scale="1,1,1") {
    alphaInputType=Bool
    rotationMode=BMM_Additive
    translationMode=BMM_Additive
    alphaBoolBlend {blendInTime=0.5, blendOption=Custom, blendOutTime=0.3}
    boneToModify {boneName=dyn_hair_tail_01}
    rotation {roll=50}
   }
   n8 = Get Head_rotation() -> (Head_rotation:double)
   n9 = "Transform (Modify) Bone - Bone: dyn_hair_tail_01"!(ComponentPose=@n7, Alpha:float=@n8, Translation="0,0,0", Rotation="0,0,-60", Scale="0.95,0.95,0.85") -> (Pose) {
    rotationMode=BMM_Additive
    scaleMode=BMM_Additive
    translationMode=BMM_Additive
    alphaBoolBlend {blendInTime=0.5, blendOption=Linear, blendOutTime=0.2}
    boneToModify {boneName=dyn_hair_tail_01}
    rotation {roll=-60}
    scale=<0.95 0.95 0.85>
   }
   n10 = "Transform (Modify) Bone - Bone: dyn_hair_tail_02"!(ComponentPose=@n9, Alpha:float=@n8, Translation="0,0,0", Rotation="0,0,15", Scale="0.9,0.9,0.75") -> (Pose) {
    rotationMode=BMM_Additive
    scaleMode=BMM_Additive
    translationMode=BMM_Additive
    alphaBoolBlend {blendInTime=0.5, blendOption=Linear, blendOutTime=0.2}
    boneToModify {boneName=dyn_hair_tail_02}
    rotation {roll=15}
    scale=<0.9 0.9 0.75>
   }
   n11 = "Transform (Modify) Bone - Bone: dyn_hair_tail_03"!(ComponentPose=@n10, Alpha:float=@n8, Translation="0,0,0", Rotation="0,0,35", Scale="0.9,0.9,0.75") -> (Pose) {
    rotationMode=BMM_Additive
    scaleMode=BMM_Additive
    translationMode=BMM_Additive
    alphaBoolBlend {blendInTime=0.5, blendOption=Linear, blendOutTime=0.2}
    boneToModify {boneName=dyn_hair_tail_03}
    rotation {roll=35}
    scale=<0.9 0.9 0.75>
   }
   n12 = Component To Local(ComponentPose=@n11)
   n13 = Use cached pose 'Aftersim'()
   n14 = Get Back_cover() -> (Back_cover:bool)
   n15 = Anim Dynamics - Bone: dyn_hair_bangs_r!(
    ComponentPose=Local To Component(LocalPose=Blend Poses by bool(BlendPose_0=@n12, BlendPose_1=@n13, bActiveValue=@n14) {blendType=Linear})
    bAlphaBoolEnabled:bool=True
    GravityOverride:Vector="0,0,-800"
    ExternalForce="0,0,0"
   ) -> (Pose) {
    alphaInputType=Bool
    angularDampingOverride=0.9999
    bOverrideAngularDamping=true
    bUseGravityOverride=true
    simulationSpace=RootRelative
    alphaBoolBlend {blendInTime=0.1, blendOption=Linear, blendOutTime=0.2}
    boundBone {boneName=dyn_hair_bangs_r}
    componentAppliedLinearAccClamp=<1500 1500 1500>
    componentLinearAccScale=<0.2 0.2 0.2>
    componentLinearVelScale=<0.3 0.3 0.3>
    gravityOverride {z=-800}
    physicsBodyDefinitions [{
      boundBone {boneName=dyn_hair_bangs_r}
      boxExtents=<10 5 5>
      constraintSetup {angularLimitsMax=<5 2.5 20>, angularLimitsMin=<-5 -5 -5>, angularTarget=<0 0 0>}
      localJointOffset=<1 0 0>
     }]
   }
   n16 = Anim Dynamics - Bone: dyn_hair_bangs_mid!(ComponentPose=@n15, bAlphaBoolEnabled:bool=True, GravityOverride:Vector="0,0,-700", ExternalForce="0,0,0") -> (Pose) {
    alphaInputType=Bool
    angularDampingOverride=1
    bOverrideAngularDamping=true
    bUseGravityOverride=true
    simulationSpace=RootRelative
    alphaBoolBlend {blendInTime=0.1, blendOption=Linear, blendOutTime=0.2}
    boundBone {boneName=dyn_hair_bangs_mid}
    componentAppliedLinearAccClamp=<2000 2000 2000>
    componentLinearAccScale=<0.2 0.2 0.2>
    componentLinearVelScale=<0.3 0.3 0.3>
    gravityOverride {z=-700}
    physicsBodyDefinitions [{
      boundBone {boneName=dyn_hair_bangs_mid}
      boxExtents=<7 7 7>
      constraintSetup {angularLimitsMax=<5 2.5 20>, angularLimitsMin=<-5 -5 -20>, angularTarget=<0 0 0>}
      localJointOffset=<2 0 0>
     }]
   }
   n17 = Anim Dynamics - Bone: dyn_hair_bangs_l!(ComponentPose=@n16, bAlphaBoolEnabled:bool=True, GravityOverride:Vector="0,0,-980", ExternalForce="0,0,0") -> (Pose) {
    alphaInputType=Bool
    angularDampingOverride=0.999
    bOverrideAngularDamping=true
    bUseGravityOverride=true
    simulationSpace=RootRelative
    alphaBoolBlend {blendInTime=0.1, blendOption=Linear, blendOutTime=0.2}
    boundBone {boneName=dyn_hair_bangs_l}
    componentAppliedLinearAccClamp=<3000 3000 3000>
    componentLinearAccScale=<0.2 0.2 0.2>
    componentLinearVelScale=<0.3 0.3 0.3>
    gravityOverride {z=-980}
    physicsBodyDefinitions [{
      boundBone {boneName=dyn_hair_bangs_l}
      boxExtents=<12 7 7>
      constraintSetup {angularLimitsMax=<5 2.5 2>, angularLimitsMin=<-5 -5 -30>, angularTarget=<0 0 0>}
      localJointOffset=<3.5 0 0>
     }]
   }
   n18 = Anim Dynamics - Bone: dyn_hair_l_01!(ComponentPose=@n17, bAlphaBoolEnabled:bool=True, GravityOverride:Vector="0,0,-900", ExternalForce="0,0,0") -> (Pose) {
    alphaInputType=Bool
    angularDampingOverride=0.999
    bOverrideAngularDamping=true
    bUseGravityOverride=true
    simulationSpace=RootRelative
    alphaBoolBlend {blendInTime=0.1, blendOption=Linear, blendOutTime=0.2}
    boundBone {boneName=dyn_hair_l_01}
    componentAppliedLinearAccClamp=<1000 1000 1000>
    componentLinearAccScale=<0.5 0.5 0.5>
    componentLinearVelScale=<0.6 0.6 0.6>
    gravityOverride {z=-900}
    physicsBodyDefinitions [{
      boundBone {boneName=dyn_hair_l_01}
      boxExtents=<10 12 12>
      constraintSetup {angularLimitsMax=<2 5 5>, angularLimitsMin=<-2 -10 -20>, angularTarget=<0 0 0>}
      localJointOffset=<1 0 0>
     }]
   }
   n19 = Anim Dynamics - Bone: dyn_hair_l_02!(ComponentPose=@n18, bAlphaBoolEnabled:bool=True, GravityOverride:Vector="0,0,-800", ExternalForce="0,0,0") -> (Pose) {
    alphaInputType=Bool
    angularDampingOverride=0.999
    bOverrideAngularDamping=true
    bUseGravityOverride=true
    simulationSpace=RootRelative
    alphaBoolBlend {blendInTime=0.1, blendOption=Linear, blendOutTime=0.2}
    boundBone {boneName=dyn_hair_l_02}
    componentAppliedLinearAccClamp=<1500 1500 1500>
    componentLinearAccScale=<0.2 0.2 0.2>
    componentLinearVelScale=<0.3 0.3 0.3>
    gravityOverride {z=-800}
    physicsBodyDefinitions [{
      boundBone {boneName=dyn_hair_l_02}
      boxExtents=<8 10 10>
      constraintSetup {angularLimitsMax=<5 5 10>, angularLimitsMin=<-5 -5 -30>, angularTarget=<0 0 0>}
      localJointOffset=<2 0 0>
     }]
   }
   n20 = Anim Dynamics - Bone: dyn_hair_r_01!(ComponentPose=@n19, bAlphaBoolEnabled:bool=True, GravityOverride:Vector="0,0,-750", ExternalForce="0,0,0") -> (Pose) {
    alphaInputType=Bool
    angularDampingOverride=0.99
    angularSpringConstant=2001
    bOverrideAngularDamping=true
    bUseGravityOverride=true
    simulationSpace=RootRelative
    alphaBoolBlend {blendInTime=0.1, blendOption=Linear, blendOutTime=0.2}
    boundBone {boneName=dyn_hair_r_01}
    componentAppliedLinearAccClamp=<1000 1000 1000>
    componentLinearAccScale=<0.5 0.5 0.5>
    componentLinearVelScale=<0.6 0.6 0.6>
    gravityOverride {z=-750}
    physicsBodyDefinitions [{
      boundBone {boneName=dyn_hair_r_01}
      boxExtents=<8 6 10>
      constraintSetup {angularLimitsMax=<5 5 30>, angularLimitsMin=<-5 -15 -15>, angularTarget=<0 0 0>}
      localJointOffset=<1 0 0>
     }]
   }
   n21 = Anim Dynamics - Bone: dyn_hair_r_02!(ComponentPose=@n20, bAlphaBoolEnabled:bool=True, GravityOverride:Vector="0,0,-800", ExternalForce="0,0,0") -> (Pose) {
    alphaInputType=Bool
    angularDampingOverride=0.999
    bOverrideAngularDamping=true
    bUseGravityOverride=true
    simulationSpace=RootRelative
    alphaBoolBlend {blendInTime=0.1, blendOption=Linear, blendOutTime=0.2}
    boundBone {boneName=dyn_hair_r_02}
    componentAppliedLinearAccClamp=<1500 1500 1500>
    componentLinearAccScale=<0.2 0.2 0.2>
    componentLinearVelScale=<0.3 0.3 0.3>
    gravityOverride {z=-800}
    physicsBodyDefinitions [{
      boundBone {boneName=dyn_hair_r_02}
      boxExtents=<8 10 10>
      constraintSetup {angularLimitsMax=<5 5 30>, angularLimitsMin=<-5 -5 -10>, angularTarget=<0 0 0>}
      localJointOffset=<3 0 0>
     }]
   }
   n22 = "Transform (Modify) Bone - Bone: dyn_hair_tail_01"!(ComponentPose=@n21, bAlphaBoolEnabled:bool=@n2, Translation="0,0,0", Rotation="0,0,20", Scale="1,1,1") -> (Pose) {
    alphaInputType=Bool
    rotationMode=BMM_Additive
    translationMode=BMM_Additive
    alphaBoolBlend {blendInTime=0.7, blendOption=Custom, blendOutTime=0.1}
    boneToModify {boneName=dyn_hair_tail_01}
    rotation {roll=20}
   }
   n23 = "Transform (Modify) Bone - Bone: dyn_hair_tail_02"!(ComponentPose=@n22, bAlphaBoolEnabled:bool=@n2, Translation="0,0,0", Rotation="0,0,20", Scale="1,1,1") -> (Pose) {
    alphaInputType=Bool
    rotationMode=BMM_Additive
    translationMode=BMM_Additive
    alphaBoolBlend {blendInTime=0.7, blendOption=Custom, blendOutTime=0.1}
    boneToModify {boneName=dyn_hair_tail_02}
    rotation {roll=20}
   }
   n24 = "Transform (Modify) Bone - Bone: dyn_hair_tail_01"(ComponentPose=@n23, bAlphaBoolEnabled=@n1, Translation="0,0,0", Rotation="0,0,10", Scale="1,1,1") {
    alphaInputType=Bool
    rotationMode=BMM_Additive
    translationMode=BMM_Additive
    alphaBoolBlend {blendInTime=0.5, blendOption=Custom, blendOutTime=0.3}
    boneToModify {boneName=dyn_hair_tail_01}
    rotation {roll=10}
   }
   n25 = "Transform (Modify) Bone - Bone: dyn_hair_tail_02"(ComponentPose=@n24, bAlphaBoolEnabled=@n1, Translation="0,0,0", Rotation="0,0,20", Scale="1,1,1") {
    alphaInputType=Bool
    rotationMode=BMM_Additive
    translationMode=BMM_Additive
    alphaBoolBlend {blendInTime=0.5, blendOption=Custom, blendOutTime=0.3}
    boneToModify {boneName=dyn_hair_tail_02}
    rotation {roll=20}
   }
   n26 = Get bIsSkydiveDiveMode()
   n27 = "Transform (Modify) Bone - Bone: dyn_hair_tail_01"(ComponentPose=@n25, bAlphaBoolEnabled=@n26, Translation="0,0,0", Rotation="0,0,-40", Scale="1,1,1") {
    alphaInputType=Bool
    rotationMode=BMM_Additive
    translationMode=BMM_Additive
    alphaBoolBlend {blendInTime=0.5, blendOption=Custom, blendOutTime=0.3}
    boneToModify {boneName=dyn_hair_tail_01}
    rotation {roll=-40}
   }
   n28 = "Transform (Modify) Bone - Bone: dyn_hair_tail_02"(ComponentPose=@n27, bAlphaBoolEnabled=@n26, Translation="0,0,0", Rotation="0,0,-5", Scale="1,1,1") {
    alphaInputType=Bool
    rotationMode=BMM_Additive
    translationMode=BMM_Additive
    alphaBoolBlend {blendInTime=0.5, blendOption=Custom, blendOutTime=0.3}
    boneToModify {boneName=dyn_hair_tail_02}
    rotation {roll=-5}
   }
   Output Pose(Result=Component To Local(ComponentPose=@n28)) {layerGroup=None}
   n29 = Copy Pose From Mesh() {bUseAttachedParent=true}
   Save cached pose 'Presim'(Pose=Component To Local(ComponentPose=Local To Component(LocalPose=@n29))) {cacheName=Presim}
   n30 = Anim Dynamics - Bone: dyn_hair_tail_strand_up!(ComponentPose=Local To Component(LocalPose=Use cached pose 'Presim'()), bAlphaBoolEnabled:bool=True, ExternalForce="0,0,0") -> (Pose) {
    alphaInputType=Bool
    angularDampingOverride=1
    bOverrideAngularDamping=true
    bUseGravityOverride=true
    alphaBoolBlend {blendInTime=0.1, blendOption=Linear, blendOutTime=0.2}
    boundBone {boneName=dyn_hair_tail_strand_up}
    componentAppliedLinearAccClamp=<300 300 300>
    componentLinearAccScale=<0.2 0.2 0.2>
    componentLinearVelScale=<0.3 0.3 0.3>
    gravityOverride {z=-800}
    physicsBodyDefinitions [{
      boundBone {boneName=dyn_hair_tail_strand_up}
      boxExtents=<8 4 2>
      constraintSetup {angularLimitsMax=<5 15 15>, angularLimitsMin=<-5 -1 -5>, angularTarget=<0 0 0>}
      localJointOffset=<1 0 0>
     }]
   }
   n31 = Get Crouch_moving()
   n32 = "Transform (Modify) Bone - Bone: dyn_hair_tail_01"(ComponentPose=@n30, bAlphaBoolEnabled=@n31, Translation="0,0,0", Rotation="0,0,35", Scale="1,1,1") {
    alphaInputType=Bool
    rotationMode=BMM_Additive
    scaleMode=BMM_Additive
    translationMode=BMM_Additive
    alphaBoolBlend {blendInTime=0.1, blendOption=Linear, blendOutTime=0.2}
    boneToModify {boneName=dyn_hair_tail_01}
    rotation {roll=35}
   }
   n33 = "Transform (Modify) Bone - Bone: dyn_hair_tail_02"(ComponentPose=@n32, bAlphaBoolEnabled=@n31, Translation="0,0,0", Rotation="0,0,-15", Scale="1,1,1") {
    alphaInputType=Bool
    rotationMode=BMM_Additive
    scaleMode=BMM_Additive
    translationMode=BMM_Additive
    alphaBoolBlend {blendInTime=0.1, blendOption=Linear, blendOutTime=0.2}
    boneToModify {boneName=dyn_hair_tail_02}
    rotation {roll=-15}
   }
   n34 = "Transform (Modify) Bone - Bone: dyn_hair_tail_03"!(ComponentPose=@n33, bAlphaBoolEnabled:bool=@n31, Translation="0,0,0", Rotation="0,0,25", Scale="1,1,1") -> (Pose) {
    alphaInputType=Bool
    rotationMode=BMM_Additive
    scaleMode=BMM_Additive
    translationMode=BMM_Additive
    alphaBoolBlend {blendInTime=0.1, blendOption=Linear, blendOutTime=0.2}
    boneToModify {boneName=dyn_hair_tail_03}
    rotation {roll=25}
   }
   n35 = Anim Dynamics - Bone: dyn_hair_tail_01!(ComponentPose=@n34, bAlphaBoolEnabled:bool=True, ExternalForce="0,0,0") -> (Pose) {
    alphaInputType=Bool
    angularDampingOverride=1
    angularSpringConstant=20000
    bOverrideAngularDamping=true
    bUseGravityOverride=true
    simulationSpace=RootRelative
    alphaBoolBlend {blendInTime=0.1, blendOption=Linear, blendOutTime=0.2}
    boundBone {boneName=dyn_hair_tail_01}
    componentAppliedLinearAccClamp=<100 100 100>
    componentLinearAccScale=<0.2 0.2 0.2>
    componentLinearVelScale=<0.3 0.3 0.3>
    gravityOverride {z=-700}
    physicsBodyDefinitions [{
      boundBone {boneName=dyn_hair_tail_01}
      boxExtents=<8 8 8>
      constraintSetup {angularLimitsMax=<2 50 20>, angularLimitsMin=<-2 0 -20>, angularTarget=<0 0 0>}
      localJointOffset=<1 0 0>
     }]
    relativeSpaceBone {boneName=neck_01}
   }
   n36 = Anim Dynamics - Bone: dyn_hair_tail_02!(ComponentPose=@n35, bAlphaBoolEnabled:bool=True, ExternalForce="0,0,0") -> (Pose) {
    alphaInputType=Bool
    angularDampingOverride=1
    angularSpringConstant=20000
    bOverrideAngularDamping=true
    bUseGravityOverride=true
    simulationSpace=RootRelative
    alphaBoolBlend {blendInTime=0.1, blendOption=Linear, blendOutTime=0.2}
    boundBone {boneName=dyn_hair_tail_02}
    componentAppliedLinearAccClamp=<100 100 100>
    componentLinearAccScale=<0.2 0.2 0.2>
    componentLinearVelScale=<0.3 0.3 0.3>
    gravityOverride {z=-800}
    physicsBodyDefinitions [{
      boundBone {boneName=dyn_hair_tail_02}
      boxExtents=<10 10 10>
      constraintSetup {angularLimitsMax=<3 50 20>, angularLimitsMin=<-3 -5 -20>, angularTarget=<0 0 0>}
      localJointOffset=<1 0 0>
     }]
    relativeSpaceBone {boneName=dyn_hair_tail_01}
   }
   n37 = Anim Dynamics - Bone: dyn_hair_tail_03!(ComponentPose=@n36, bAlphaBoolEnabled:bool=True, ExternalForce="0,0,0") -> (Pose) {
    alphaInputType=Bool
    angularDampingOverride=1
    angularSpringConstant=20000
    bOverrideAngularDamping=true
    bUseGravityOverride=true
    simulationSpace=RootRelative
    alphaBoolBlend {blendInTime=0.1, blendOption=Linear, blendOutTime=0.2}
    boundBone {boneName=dyn_hair_tail_03}
    componentAppliedLinearAccClamp=<100 100 100>
    componentLinearAccScale=<0.2 0.2 0.2>
    componentLinearVelScale=<0.3 0.3 0.3>
    gravityOverride {z=-900}
    physicsBodyDefinitions [{
      boundBone {boneName=dyn_hair_tail_03}
      boxExtents=<12 12 12>
      constraintSetup {angularLimitsMax=<4 5 25>, angularLimitsMin=<-4 -5 -25>, angularTarget=<0 0 0>}
      localJointOffset=<1 0 0>
     }]
    relativeSpaceBone {boneName=dyn_hair_tail_02}
   }
   n38 = Anim Dynamics - Bone: dyn_hair_tail_04!(ComponentPose=@n37, bAlphaBoolEnabled:bool=True, ExternalForce="0,0,0") -> (Pose) {
    alphaInputType=Bool
    angularDampingOverride=0.99
    angularSpringConstant=20000
    bOverrideAngularDamping=true
    bUseGravityOverride=true
    simulationSpace=RootRelative
    alphaBoolBlend {blendInTime=0.1, blendOption=Linear, blendOutTime=0.2}
    boundBone {boneName=dyn_hair_tail_04}
    componentAppliedLinearAccClamp=<300 300 300>
    componentLinearAccScale=<0.2 0.2 0.2>
    componentLinearVelScale=<0.3 0.3 0.3>
    gravityOverride {z=-800}
    physicsBodyDefinitions [{
      boundBone {boneName=dyn_hair_tail_04}
      boxExtents=<8 15 8>
      constraintSetup {angularLimitsMax=<5 5 25>, angularLimitsMin=<-5 -5 -25>, angularTarget=<0 0 0>}
      localJointOffset=<1 0 0>
     }]
    relativeSpaceBone {boneName=dyn_hair_tail_03}
   }
   n39 = Anim Dynamics - Bone: dyn_hair_tail_05!(ComponentPose=@n38, bAlphaBoolEnabled:bool=True, ExternalForce="0,0,0") -> (Pose) {
    alphaInputType=Bool
    angularDampingOverride=0.9
    bOverrideAngularDamping=true
    bUseGravityOverride=true
    simulationSpace=RootRelative
    alphaBoolBlend {blendInTime=0.1, blendOption=Linear, blendOutTime=0.2}
    boundBone {boneName=dyn_hair_tail_05}
    componentAppliedLinearAccClamp=<1000 1000 1000>
    componentLinearAccScale=<0.2 0.2 0.2>
    componentLinearVelScale=<0.3 0.3 0.3>
    gravityOverride {z=-900}
    physicsBodyDefinitions [{
      boundBone {boneName=dyn_hair_tail_05}
      boxExtents=<5 5 5>
      constraintSetup {angularLimitsMax=<15 10 -5>, angularLimitsMin=<-15 -10 -45>, angularTarget=<0 0 0>}
      localJointOffset=<1 0 0>
     }]
   }
   n40 = Anim Dynamics - Bone: dyn_hair_tail_06!(ComponentPose=@n39, bAlphaBoolEnabled:bool=True, ExternalForce="0,0,0") -> (Pose) {
    alphaInputType=Bool
    angularDampingOverride=0.9
    bOverrideAngularDamping=true
    bUseGravityOverride=true
    simulationSpace=RootRelative
    alphaBoolBlend {blendInTime=0.1, blendOption=Linear, blendOutTime=0.2}
    boundBone {boneName=dyn_hair_tail_06}
    componentAppliedLinearAccClamp=<1000 1000 1000>
    componentLinearAccScale=<0.2 0.2 0.2>
    componentLinearVelScale=<0.3 0.3 0.3>
    gravityOverride {z=-900}
    physicsBodyDefinitions [{
      boundBone {boneName=dyn_hair_tail_06}
      boxExtents=<5 5 5>
      constraintSetup {angularLimitsMax=<15 5 45>, angularLimitsMin=<-15 -5 0>, angularTarget=<0 0 0>}
      localJointOffset=<1 0 0>
     }]
   }
   n41 = Anim Dynamics - Bone: dyn_hair_tail_07!(ComponentPose=@n40, bAlphaBoolEnabled:bool=True, ExternalForce="0,0,0") -> (Pose) {
    alphaInputType=Bool
    angularDampingOverride=0.999
    bOverrideAngularDamping=true
    bUseGravityOverride=true
    simulationSpace=RootRelative
    alphaBoolBlend {blendInTime=0.1, blendOption=Linear, blendOutTime=0.2}
    boundBone {boneName=dyn_hair_tail_07}
    componentAppliedLinearAccClamp=<500 500 500>
    componentLinearAccScale=<0.2 0.2 0.2>
    componentLinearVelScale=<0.3 0.3 0.3>
    gravityOverride {z=-900}
    physicsBodyDefinitions [{
      boundBone {boneName=dyn_hair_tail_07}
      boxExtents=<8 8 8>
      constraintSetup {angularLimitsMax=<10 30 20>, angularLimitsMin=<-10 0 -20>, angularTarget=<0 0 0>}
      localJointOffset=<1 0 0>
     }]
   }
   n42 = RigidBody!(
    ComponentPose=Local To Component(LocalPose=Use cached pose 'Presim'())
    bAlphaBoolEnabled:bool=Get RB_moving() -> (RB_moving:bool)
    OverrideWorldGravity:Vector="0,0,-980"
    ExternalForce
   ) -> (Pose) {
    alphaInputType=Bool
    bClampLinearTranslationLimitToRefPose=true
    bOverrideWorldGravity=true
    overridePhysicsAsset=/BRCosmetics/Characters/Player/Female/Medium/Bodies/F_MED_BrightIon/Meshes/Parts/F_MED_BrightIon_FaceAcc_Physics
    simulationSpace=BaseBoneSpace
    alphaBoolBlend {blendInTime=0.1, blendOption=Linear, blendOutTime=0.1}
    baseBoneRef {boneName=neck_01}
    componentAppliedLinearAccClamp=<3000 3000 3000>
    componentLinearAccScale=<0.2 0.3 0.3>
    componentLinearVelScale=<0.4 0.4 0.4>
    overrideWorldGravity {z=-980}
   }
   n43 = RigidBody(ComponentPose=@n42, bAlphaBoolEnabled=Get Jump_loop() -> (Jump_loop:bool), OverrideWorldGravity="0,0,-980") {
    alphaInputType=Bool
    bClampLinearTranslationLimitToRefPose=true
    bOverrideWorldGravity=true
    overridePhysicsAsset=&F_MED_BrightIon_FaceAcc_Physics
    simulationSpace=BaseBoneSpace
    alphaBoolBlend {blendInTime=0.1, blendOption=Linear, blendOutTime=0.1}
    baseBoneRef {boneName=head}
    componentAppliedLinearAccClamp=<3000 3000 3000>
    componentLinearAccScale=<0.3 0.3 0.3>
    componentLinearVelScale=<0.4 0.4 0.4>
    overrideWorldGravity {z=-980}
   }
   n44 = RigidBody(ComponentPose=@n43, bAlphaBoolEnabled=Get bIsSkyDiving(), OverrideWorldGravity="0,0,-980") {
    alphaInputType=Bool
    bClampLinearTranslationLimitToRefPose=true
    bOverrideWorldGravity=true
    overridePhysicsAsset=&F_MED_BrightIon_FaceAcc_Physics
    simulationSpace=BaseBoneSpace
    alphaBoolBlend {blendInTime=0.1, blendOption=Linear, blendOutTime=0.1}
    baseBoneRef {boneName=head}
    componentAppliedLinearAccClamp=<1000 1000 1000>
    componentLinearAccScale=<0.1 0.1 0.1>
    componentLinearVelScale=<0.2 0.2 0.2>
    overrideWorldGravity {z=-980}
   }
   Save cached pose 'Aftersim'(Pose=Blend Poses by bool(BlendPose_0=Component To Local(ComponentPose=@n44), BlendPose_1=Component To Local(ComponentPose=@n41), bActiveValue=Get bRigidBodyEnabled()) {blendType=Linear}) {cacheName=Aftersim}
  ]}
 EventGraph {nodes [
   n1 = Event BlueprintUpdateAnimation()
   > n2 = Sequence!(execute) -> (then_0, then_1, then_2:Exec)
   > n3 = "Gravity Override State Function (Parameter Asset) (Nativized)"(
    CharacterPartAnimInstance=Self-Reference()
    GameTimeInSeconds=GetGameTimeInSeconds()
    Parameters=/BRCosmetics/Characters/Player/Female/Medium/Bodies/F_MED_FermiumStraw_Rain/Meshes/Parts/F_MED_FermiumStraw_Rain_Parameters
   )
   > Set GravityOverride(GravityOverride=@n3)
   > Set toggleAnimDynamics(toggleAnimDynamics:bool=NOT Boolean(A=Get isOnZipline() -> (isOnZipline:bool))) -> (Output_Get:bool)
   n4 = float * float(
    A=GetAlphaFromReferenceTransformDelta(SkeletalMeshComponent=GetOwnerSkeletalMeshComponent(), BoneName=head, BaseName=neck_02, InputRange=(X=0,Y=0), TwistAxis="1,0,0").RawOutput
    B=-0.05
   ) {is=title K2Node_PromotableOperator}
   Set Head_rotation(
    execute=@n2.then_1
    Head_rotation:double=FInterpTo(Current=Get Head_rotation(), Target="Clamp (Float)"(Value=@n4, Min=0, Max=1), DeltaTime=@n1.DeltaTimeX, InterpSpeed=5)
   ) -> (Output_Get:double)
   n5 = Get bIsSkyDiving()
   n6 = CheckForBackbling()
   Set Back_cover(
    execute=@n2.then_2
    Back_cover:bool=AND Boolean(A=OR Boolean!(A=@n6, B=@n6.bIsBackblingACape, C:bool) -> (ReturnValue), B=NOT Boolean(A=@n5))
   ) -> (Output_Get:bool)
   n7 = Get bIsSkyDiving()
   > Set jump_falling(jump_falling:bool=AND Boolean(A=Get bIsFalling(), B=NOT Boolean(A=@n7))) -> (Output_Get:bool)
   > Set Crouch_moving(Crouch_moving:bool=AND Boolean(A=Get bIsCrouching(), B=Get bIsMoving2D())) -> (Output_Get:bool)
   > Set Not Sky Diving(Not Sky Diving:bool=NOT Boolean(A=Get bIsSkyDiving())) -> (Output_Get:bool)
   > Set Jump_loop(Jump_loop:bool=OR Boolean!(A=Get jump_falling(), B=Get bIsJumping()) -> (ReturnValue)) -> (Output_Get:bool)
   n8 = Get Jump_loop()
   > Set RB_moving(RB_moving:bool=OR Boolean(A=Get Not Sky Diving() -> (Not Sky Diving:bool), B=NOT Boolean(A=@n8))) -> (Output_Get:bool)
   Event BlueprintInitializeAnimation()
   > Set SkelMesh(SkelMesh=GetOwnerSkeletalMeshComponent())
   n9 = Event BlueprintProcessAnimInputEvent()
   > n10 = Branch(Condition=IsNamed(self=@n9.AnimInputEvent, InEventName=aie.Player.Zipline.Entered))
   > Set isOnZipline(isOnZipline:bool=1) -> (Output_Get:bool)
   Branch(execute=@n10.else, Condition=IsNamed(self=@n9.AnimInputEvent, InEventName=aie.Player.Zipline.Exited))
   > Set isOnZipline(isOnZipline=false)
  ]}
]
variables [
 Back_cover {type=boolean}
 Crouch_moving {type=boolean}
 GravityOverride {type=Vector}
 Head_rotation {type=number}
 isOnZipline {type=boolean}
 jump_falling {type=boolean}
 Jump_loop {type=boolean}
 Not Sky Diving {type=boolean}
 RB_moving {type=boolean}
 SkelMesh {type=SkeletalMeshComponent}
 toggleAnimDynamics {type=boolean}
]
