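# Physics asset for the FaceAcc mesh part of the F_MED_FermiumStraw_Dust female character, setting up a three-segment hair tail on the head.
# The tail chains two capsules and a larger end sphere with limited joints: tight swing and wide twist at the head, wider swing and almost no twist further down.
# Capsules on head, neck, upper spine and left clavicle add body volumes; head-to-spine joints are free, and the clavicle has no constraint.
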PhysicsAsset /BRCosmetics/Characters/Player/Female/Medium/Bodies/F_MED_FermiumStraw_Dust/Meshes/Parts/F_MED_FermiumStraw_Dust_FaceAcc_Physics
bodies {
 clavicle_l [{shapeType=Capsule, length=15, radius=4, shapeName=clavicle_l_capsule, center=<10.2032 -1.6719 0>, rotation=<rot -90 0 -10.0006>}]
 dyn_hair_tail_01 [{shapeType=Capsule, length=7, radius=3.5, shapeName=dyn_hair_tail_01_capsule, center=<3.456 0 0>, rotation=<rot 90 180 -90>}]
 dyn_hair_tail_02 [{shapeType=Capsule, length=2, radius=2, shapeName=dyn_hair_tail_01_capsule, center=<4.9814 0 0>, rotation=<rot 90 180 180>}]
 dyn_hair_tail_03 [{shapeType=Sphere, radius=6.5, shapeName=dyn_hair_tail_03_sphere, center=<6.3463 0 -0.3613>}]
 head [{shapeType=Capsule, length=5, radius=9, shapeName=head_capsule, center=<3.6247 0.8985 0>, rotation=<rot 90 180 -159.9994>}]
 neck_01 [{shapeType=Capsule, length=0, radius=5, shapeName=neck_01_capsule}]
 neck_02 [{shapeType=Capsule, length=0, radius=5, shapeName=neck_02_capsule}]
 spine_05 [{shapeType=Capsule, length=15, radius=11.5, shapeName=spine_05_capsule, center=<2.5 0 0>}]
}
constraints [
 {bone1Name=dyn_hair_tail_01, bone2Name=head, swing1LimitDegrees=5, swing2LimitDegrees=10, twistLimitDegrees=60}
 {bone1Name=dyn_hair_tail_02, bone2Name=dyn_hair_tail_01, swing1LimitDegrees=30, swing2LimitDegrees=30, twistLimitDegrees=1}
 {bone1Name=dyn_hair_tail_03, bone2Name=dyn_hair_tail_02, swing1LimitDegrees=30, swing2LimitDegrees=30, twistLimitDegrees=1}
 {
  bone1Name=head
  bone2Name=neck_02
  swing1LimitDegrees=45
  swing1Motion=Free
  swing2LimitDegrees=45
  swing2Motion=Free
  twistLimitDegrees=45
  twistMotion=Free
 }
 {
  bone1Name=neck_02
  bone2Name=neck_01
  swing1LimitDegrees=45
  swing1Motion=Free
  swing2LimitDegrees=45
  swing2Motion=Free
  twistLimitDegrees=45
  twistMotion=Free
 }
 {
  bone1Name=neck_01
  bone2Name=spine_05
  swing1LimitDegrees=45
  swing1Motion=Free
  swing2LimitDegrees=45
  swing2Motion=Free
  twistLimitDegrees=45
  twistMotion=Free
 }
]
